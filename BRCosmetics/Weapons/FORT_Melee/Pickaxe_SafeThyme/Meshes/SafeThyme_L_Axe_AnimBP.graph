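# Animation Blueprint for the left SafeThyme pickaxe axe, adding physics sway to its tape_01_l bone.
# The cached input pose feeds two Anim Dynamics angular-spring setups, picked by bIsInFrontEnd.
# The front-end setup has a stiffer spring, smaller box and lower component motion scales.
# Update Animation branches on the same flag to set SimlationSpace, which no node visibly reads; nothing here sets the flag.
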AnimBlueprint /BRCosmetics/Weapons/FORT_Melee/Pickaxe_SafeThyme/Meshes/SafeThyme_L_Axe_AnimBP
target_skeleton=&/SafeThyme_Axe_Skeleton
graphs [
 AnimGraph {nodes [
   n1 = Anim Dynamics - Bone: tape_01_l!(
    SimulationSpace:AnimPhysSimSpaceType Enum
    ComponentPose=Local To Component(LocalPose=Use cached pose 'Main pose'())
    Alpha=1
    GravityOverride:Vector="0,0,0"
    ExternalForce="0,0,0"
   ) -> (Pose) {
    angularDampingOverride=0.999
    angularSpringConstant=2800
    bAngularSpring=true
    bGravityOverrideInSimSpace=true
    bOverrideAngularDamping=true
    bOverrideLinearDamping=true
    bUseGravityOverride=true
    linearDampingOverride=1
    linearSpringConstant=100
    alphaBoolBlend {blendOption=Linear}
    boundBone {boneName=tape_01_l}
    componentAppliedLinearAccClamp=<3000 3000 2500>
    componentLinearAccScale=<0.6 0.6 0.6>
    componentLinearVelScale=<0.6 0.6 0.6>
    physicsBodyDefinitions [{
      boundBone {boneName=tape_01_l}
      boxExtents=<20 20 20>
      constraintSetup {coneAngle=1, angularLimitsMax=<22 33 90>, angularLimitsMin=<-22 -33 -77>, angularTarget=<1 0 0>}
      localJointOffset=<1.3 0 0>
     }]
    simSpaceSettings {maxAngularAcceleration=35, maxAngularVelocity=7, simSpaceAngularAlpha=1}
   }
   n2 = Anim Dynamics - Bone: tape_01_l(ComponentPose=Local To Component(LocalPose=Use cached pose 'Main pose'()), Alpha=1, GravityOverride="0,0,0", ExternalForce="0,0,0") {
    angularDampingOverride=0.9999
    angularSpringConstant=3800
    bAngularSpring=true
    bGravityOverrideInSimSpace=true
    bOverrideAngularDamping=true
    bUseGravityOverride=true
    linearDampingOverride=1
    linearSpringConstant=100
    alphaBoolBlend {blendOption=Linear}
    boundBone {boneName=tape_01_l}
    componentAppliedLinearAccClamp=<3000 3000 2500>
    componentLinearAccScale=<0.3 0.3 0.3>
    componentLinearVelScale=<0.3 0.3 0.3>
    physicsBodyDefinitions [{
      boundBone {boneName=tape_01_l}
      boxExtents=<15 15 15>
      constraintSetup {coneAngle=1, angularLimitsMax=<22 33 90>, angularLimitsMin=<-22 -33 -77>, angularTarget=<1 0 0>}
      localJointOffset=<1.3 0 0>
     }]
    simSpaceSettings {maxAngularAcceleration=35, maxAngularVelocity=7, simSpaceAngularAlpha=1}
   }
   Output Pose(Result=Blend Poses by bool(BlendPose_0=Component To Local(ComponentPose=@n2), BlendPose_1=Component To Local(ComponentPose=@n1), bActiveValue=Get bIsInFrontEnd()))
   Save cached pose 'Main pose'(Pose=Input Pose()) {cacheName=Main pose}
  ]}
 EventGraph {nodes [
   Event BlueprintUpdateAnimation()
   > n1 = Branch(Condition=Get bIsInFrontEnd())
   > Set SimlationSpace(SimlationSpace:AnimPhysSimSpaceType Enum=Component) -> (Output_Get:AnimPhysSimSpaceType Enum)
   Set SimlationSpace(execute=@n1.else)
   TryGetPawnOwner()
  ]}
]
variables [
 bIsInFrontEnd {type=boolean}
 SimlationSpace {type=AnimPhysSimSpaceType}
]
